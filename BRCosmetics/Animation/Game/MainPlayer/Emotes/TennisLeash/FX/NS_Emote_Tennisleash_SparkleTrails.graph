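# Niagara system for the TennisLeash emote that emits warm HDR sparkle-sprite trails from a skeletal mesh.
# Two near-identical emitters differ mainly in cost: SparklesHigh is a dense GPU sim, SparklesLow a sparser non-GPU one.
# Sparkles spawn around the mesh with partly inherited velocity, and a kill check compares sampled mesh speed to a threshold.
# Upward gravity, curl noise and drag drive motion, while curves and a randomized waveform create fading twinkle.
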NiagaraSystem /BRCosmetics/Animation/Game/MainPlayer/Emotes/TennisLeash/FX/NS_Emote_Tennisleash_SparkleTrails
dependencies {
 usedDataInterfaces [&NiagaraDataInterfaceCurve, &NiagaraDataInterfaceSkeletalMesh]
 usedDynamicInputs []
 usedModules [
  &self:SystemSpawnScript.SystemScriptSource.SystemScriptGraph.NiagaraNodeAssignment_0.SetVariables_#1
  &SystemState
 ]
}
emitters {
 SparklesHigh {
  simTarget=GPUComputeSim
  emitterUpdateScript [
   &EmitterState()
   &SpawnRate(SpawnRate=700, Use Spawn Probability=false)
  ]
  particleSpawnScript [
   &InitializeParticle(
    Color=<rgba 5 3.3942 2.1582 1>
    Lifetime=&UniformRangedFloat(Minimum=0.4, Maximum=1)
    Position Offset=<0 0 10>
    Sprite Size Mode=::NewEnumerator3
    Uniform Sprite Size=&UniformRangedFloat(Minimum=10, Maximum=30)
    UsePositionOffset=true
    Write Material Random=true
   ) {module=InitializeParticle001}
   &PartitionParticles(
    ID to Partition:NiagaraInt32=&ReturnExecIndex()
    Number Of Partitions=5
    Partition Mode:ENiagara_PartitionMode=::NewEnumerator1
   )
   &SkeletalMeshLocation(
    Bone Sampling Mode=::NewEnumerator0
    "Inherit Velocity (Scale)"=0.25
    MaskSpawnGroup=false
    OverrideRandomSeed=false
    Skeletal Mesh=$System.Skeletal Mesh
    Velocity Sampling=::NewEnumerator0
    VelocitySpeedLimit=false
   )
   &ShapeLocation(Box Size=<25 25 25>, MaskSpawnGroup=false, Override Seed=false, Shape Primitive=::NewEnumerator2, Surface Only=false, UseOwnerScale=true)
   &ApplyInitialForces()
  ]
  particleUpdateScript [
   &KillParticles(Kill Particles=&SetBoolByFloatComparison(A=50, B=&VectorLength(Vector=$Particles.SkeletalMeshLocation.SampledVelocity)))
   &ParticleState()
   &GravityForce(Gravity=<0 0 150>)
   /Niagara/Modules/Update/Forces/CurlNoiseForce(Noise Frequency=0.02, Noise Quality / Cost=::NewEnumerator0, Noise Strength=1000, Pan Noise=false)
   &Drag(Drag=3, Use Linear Drag=true, Use Rotational Drag=false)
   &SolveForcesAndVelocity(Clamp Velocity=false, Limit Acceleration=false)
   &ScaleColor(
    Scale Alpha=&FloatFromCurve(FloatCurve=DataInterface() {propertyValues {Curve {keys [<at -0.0092 0.0144>,<at 0.0967 0.9767>,<at 1.0011 0.002>]}}})
    ScaleA=true
    ScaleRGB=true
   )
   &ScaleSpriteSize(Scale Factor=&Vector2DFromFloat(Value=&FloatFromCurve(
      FloatCurve=DataInterface() {propertyValues {Curve {keys [<at -0.0046 0>,<at 0.1746 0.9952>,<at 1.0011 0.002>]}}}
      Scale Curve=&Waveform(
       Global Input Position=&RandomRangeFloat(Minimum=0, Maximum=1)
       "Waveform [1]"=::NewEnumerator12
       "[1] Frequency"=10
       "[1] Amplitude Min/Max"=<0 1>
      )
     )))
  ]
  rendererClasses [&NiagaraSpriteRendererProperties]
  renderers [{rendererClass=&NiagaraSpriteRendererProperties}]
 }
 SparklesLow {
  emitterUpdateScript [
   &EmitterState()
   &SpawnRate(SpawnRate=100, Use Spawn Probability=false)
  ]
  particleSpawnScript [
   &InitializeParticle(
    Color=<rgba 5 3.3942 2.1582 1>
    Lifetime=&UniformRangedFloat(Minimum=0.4, Maximum=1)
    Position Offset=<0 0 10>
    Sprite Size Mode=::NewEnumerator3
    Uniform Sprite Size=&UniformRangedFloat(Minimum=10, Maximum=30)
    UsePositionOffset=true
    Write Material Random=true
   ) {module=InitializeParticle001}
   &PartitionParticles(
    ID to Partition:NiagaraInt32=&ReturnExecIndex()
    Number Of Partitions=5
    Partition Mode:ENiagara_PartitionMode=::NewEnumerator1
   )
   &SkeletalMeshLocation(
    Bone Sampling Mode=::NewEnumerator0
    "Inherit Velocity (Scale)"=0.25
    MaskSpawnGroup=false
    OverrideRandomSeed=false
    Skeletal Mesh=$System.Skeletal Mesh
    Velocity Sampling=::NewEnumerator0
    VelocitySpeedLimit=false
   )
   &ShapeLocation(Box Size=<25 25 25>, MaskSpawnGroup=false, Override Seed=false, Shape Primitive=::NewEnumerator2, Surface Only=false, UseOwnerScale=true)
   &ApplyInitialForces()
  ]
  particleUpdateScript [
   &KillParticles(Kill Particles=&SetBoolByFloatComparison(A=50, B=&VectorLength(Vector=$Particles.SkeletalMeshLocation.SampledVelocity)))
   &ParticleState()
   &GravityForce(Gravity=<0 0 150>)
   &CurlNoiseForce(Noise Frequency=0.02, Noise Quality / Cost=::NewEnumerator0, Noise Strength=1000, Pan Noise=false)
   &Drag(Drag=3, Use Linear Drag=true, Use Rotational Drag=false)
   &SolveForcesAndVelocity(Clamp Velocity=false, Limit Acceleration=false)
   &ScaleColor(
    Scale Alpha=&FloatFromCurve(FloatCurve=DataInterface() {propertyValues {Curve {keys [<at -0.0092 0.0144>,<at 0.0967 0.9767>,<at 1.0011 0.002>]}}})
    ScaleA=true
    ScaleRGB=true
   )
   &ScaleSpriteSize(Scale Factor=&Vector2DFromFloat(Value=&FloatFromCurve(
      FloatCurve=DataInterface() {propertyValues {Curve {keys [<at -0.0046 0>,<at 0.1746 0.9952>,<at 1.0011 0.002>]}}}
      Scale Curve=&Waveform(
       Global Input Position=&RandomRangeFloat(Minimum=0, Maximum=1)
       "Waveform [1]"=::NewEnumerator12
       "[1] Frequency"=10
       "[1] Amplitude Min/Max"=<0 1>
      )
     )))
  ]
  rendererClasses [&NiagaraSpriteRendererProperties]
  renderers [{rendererClass=&NiagaraSpriteRendererProperties}]
 }
}
summary {
 systemName=NS_Emote_Tennisleash_SparkleTrails
 emitters [
  {emitterName=SparklesHigh, simTarget=GPUComputeSim, rendererClasses [&NiagaraSpriteRendererProperties]}
  {emitterName=SparklesLow, rendererClasses [&NiagaraSpriteRendererProperties]}
 ]
 userVariables []
}
